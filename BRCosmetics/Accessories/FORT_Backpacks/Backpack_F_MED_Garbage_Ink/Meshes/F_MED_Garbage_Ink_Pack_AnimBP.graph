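# Animation Blueprint for the Garbage Ink backpack that procedurally shakes its motor and pipes like a running engine.
# Each update, game-time sine waves, partly randomized, set the rotations and pipe scale fed into a chain of Modify Bone nodes.
# An increase in the watched Athena kill stat boosts rotation shake for three seconds.
# The output uses the cached vibration pose while accelerating, otherwise a Brake variant with the handle turned.
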AnimBlueprint /BRCosmetics/Accessories/FORT_Backpacks/Backpack_F_MED_Garbage_Ink/Meshes/F_MED_Garbage_Ink_Pack_AnimBP
target_skeleton=&/F_MED_Garbage_Ink_Pack_Skeleton
graphs [
 AnimGraph {nodes [
   Output Pose(Result=Blend Poses by bool(BlendPose_0=Use cached pose 'Vibration'(), BlendPose_1=Use cached pose 'Brake'(), bActiveValue=Get bIsAccelerating2D()) {blendType=Linear}) {layerGroup=None}
   n1 = "Transform (Modify) Bone - Bone: dyn_pipe_scale_jnt"(
    ComponentPose=Local To Component(LocalPose=Input Pose())
    Alpha=1
    Translation="0,0,0"
    Rotation="0,0,0"
    Scale=Get PipeScale() -> (PipeScale:Vector)
   ) {scaleMode=BMM_Replace, scaleSpace=BCS_BoneSpace, alphaBoolBlend {blendOption=Linear}, boneToModify {boneName=dyn_pipe_scale_jnt}}
   n2 = "Transform (Modify) Bone - Bone: dyn_twinMotor_jnt"(ComponentPose=@n1, Alpha=1, Translation="0,0,0", Rotation=Get MotorRotate() -> (MotorRotate:Rotator), Scale="1,1,1") {
    rotationMode=BMM_Replace
    alphaBoolBlend {blendOption=Linear}
    alphaScaleBiasClamp {bInterpResult=true}
    boneToModify {boneName=dyn_twinMotor_jnt}
   }
   n3 = Get MainMassRotate() -> (MainMassRotate:Rotator)
   n4 = "Transform (Modify) Bone - Bone: dyn_masterRotater_jnt"(ComponentPose=@n2, Alpha=1, Translation="0,0,0", Rotation=@n3, Scale="1,1,1") {
    rotationMode=BMM_Replace
    alphaBoolBlend {blendOption=Linear}
    alphaScaleBiasClamp {bInterpResult=true}
    boneToModify {boneName=dyn_masterRotater_jnt}
   }
   n5 = "Transform (Modify) Bone - Bone: dyn_pipe_jnt_1"(ComponentPose=@n4, Alpha=1, Translation="0,0,0", Rotation=Get PipeRotate() -> (PipeRotate:Rotator), Scale="1,1,1") {
    rotationMode=BMM_Replace
    rotationSpace=BCS_BoneSpace
    alphaBoolBlend {blendOption=Linear}
    alphaScaleBiasClamp {bInterpResult=true}
    boneToModify {boneName=dyn_pipe_jnt_1}
   }
   n6 = "Transform (Modify) Bone - Bone: dyn_pipe_jnt_2"(ComponentPose=@n5, Alpha=1, Translation="0,0,0", Rotation=@n3, Scale="1,1,1") {rotationMode=BMM_Replace, rotationSpace=BCS_BoneSpace, alphaBoolBlend {blendOption=Linear}, boneToModify {boneName=dyn_pipe_jnt_2}}
   n7 = "Transform (Modify) Bone - Bone: dyn_pipe_jnt_2"(ComponentPose=@n6, Alpha=1, Translation="0,0,0", Rotation=@n3, Scale="1,1,1") {rotationMode=BMM_Replace, rotationSpace=BCS_BoneSpace, alphaBoolBlend {blendOption=Linear}, boneToModify {boneName=dyn_pipe_jnt_2}}
   n8 = "Transform (Modify) Bone - Bone: dyn_handle_jnt"(ComponentPose=@n7, Alpha=1, Translation="0,0,0", Rotation="0,0,0", Scale="1,1,1") {rotationMode=BMM_Replace, rotationSpace=BCS_BoneSpace, alphaBoolBlend {blendOption=Linear}, boneToModify {boneName=dyn_handle_jnt}}
   Save cached pose 'Vibration'(Pose=Component To Local(ComponentPose=@n8)) {cacheName=Vibration}
   n9 = "Transform (Modify) Bone - Bone: dyn_handle_jnt"(ComponentPose=Local To Component(LocalPose=Use cached pose 'Vibration'()), Alpha=1, Translation="0,0,0", Rotation="0,12,0", Scale="1,1,1") {
    rotationMode=BMM_Replace
    rotationSpace=BCS_BoneSpace
    alphaBoolBlend {blendOption=Linear}
    boneToModify {boneName=dyn_handle_jnt}
    rotation {yaw=12}
   }
   Save cached pose 'Brake'(Pose=Component To Local(ComponentPose=@n9)) {cacheName=Brake}
  ]}
 EventGraph {nodes [
   Event BlueprintUpdateAnimation()
   > Get!(execute:Exec) -> (then:Exec, else:Exec, Skel Mesh:Skeletal Mesh Component Object Reference)
   > Get!(execute) -> (then, else, Pawn:Fort Player Pawn Object Reference)
   > n1 = Sequence!(execute) -> (then_0, then_1, then_2:Exec, then_3:Exec, then_4:Exec)
   n2 = GetGameTimeInSeconds()
   n3 = "Sin (Radians)"(A=float * float(A=float * float(A=1, B=20), B=@n2))
   n4 = RandomFloatInRange(
    Min=float + float(A=float / float(A=@n3, B=10) {is=title K2Node_CallFunction}, B=1) {is=title K2Node_CommutativeAssociativeBinaryOperator}
    Max=0.9
   )
   > Set PipeScale(PipeScale:Vector=MakeVector(X=float + float(A=0, B=1), Y=@n4, Z=@n4)) -> (Output_Get:Vector)
   n5 = "Sin (Radians)"(A=float * float(A=float * float(A=1, B=60), B=@n2))
   n6 = SelectFloat(A=1, B=0.3, bPickA=Get Active Kill() -> (Active Kill:bool))
   Set MotorRotate(execute=@n1.then_1, MotorRotate:Rotator=MakeRotator(Pitch=float * float(A=@n5, B=@n6))) -> (Output_Get:Rotator)
   n7 = float * float(A=1, B=60)
   n8 = float * float(A="Sin (Radians)"(A=float * float(A=@n7, B=@n2)), B=SelectFloat(A=0.6, B=0.3, bPickA=Get Active Kill()))
   Set MainMassRotate(execute=@n1.then_2, MainMassRotate:Rotator=MakeRotator(Pitch=@n8, Yaw=@n8)) -> (Output_Get:Rotator)
   n9 = Get Active Kill()
   n10 = float * float(A=float * float(A=1, B=60), B=@n2)
   n11 = RandomFloatInRange(Min=float * float(A="Sin (Radians)"(A=@n10), B=SelectFloat(A=2, B=0.6, bPickA=@n9)), Max=1.7)
   Set PipeRotate(execute=@n1.then_3, PipeRotate:Rotator=MakeRotator(Pitch=@n11, Yaw=@n11)) -> (Output_Get:Rotator)
   n12 = WatchCosmeticStat!(
    execute:Exec=@n1.then_4
    Pawn:Fort Player Pawn Object Reference=Get Pawn() -> (Pawn:Fort Player Pawn Object Reference)
    StatTag:Fort Stat Manager Tag Structure=(TagName="StatManager.AthenaKills")
    DefaultValue:int=0
    bBroadcastInitialValue:bool=true
    InitialValueBroadcastDelay:float=-1
    ChangeValueBroadcastDelay:float=-1
    bConsumeWaitingNotifications:bool=true
   ) -> (then:Exec, OnStatChanged:Exec, StatName:Name, StatValue:int) {is=title K2Node_AsyncAction}
   Set Did_Kill(execute=@n12.OnStatChanged, Did_Kill:bool=false) -> (Output_Get:bool)
   > n13 = Set Kill Count(Kill Count:double="To Float (Integer)"(InInt=@n12.StatValue)) -> (Output_Get:double)
   > Branch(Condition="float > float"(A=@n13, B=Get Last Kill Count() -> (Last Kill Count:double)) {is=title K2Node_CallFunction})
   > Set Active Kill(Active Kill:bool=true) -> (Output_Get:bool)
   > Set Last Kill Count(Last Kill Count:double=@n13) -> (Output_Get:double)
   > Delay(Duration=3)
   > Set Active Kill(Active Kill=false)
   Event BlueprintInitializeAnimation()
   > Set Skel Mesh(Skel Mesh=GetOwnerSkeletalMeshComponent())
   > n14 = Cast To FortPlayerPawn(Object=TryGetFortPawn())
   > Set Pawn(Pawn:Fort Player Pawn Object Reference=@n14) -> (Output_Get:Fort Player Pawn Object Reference)
   GetOwningActor()
   Event OnCharacterCustomizationLoaded()
  ]}
]
variables [
 Active Kill {type=boolean}
 Did_Kill {type=boolean}
 Kill Count {type=number}
 Last Kill Count {type=number}
 MainMassRotate {type=Rotator}
 MotorRotate {type=Rotator}
 Pawn {type=FortPlayerPawn}
 PipeRotate {type=Rotator}
 PipeScale {type=Vector}
 Skel Mesh {type=SkeletalMeshComponent}
]
